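# Physics asset for the female medium RedOasis ClubDark body cosmetic, covering only the pelvis, thighs and left and right dynamic shorts-hem bones.
# Each of the five bones carries a single capsule body.
# Constraints chain each hem to its thigh and each thigh to the pelvis.
# The hem joints have tight, uneven swing and twist limits, keeping shorts motion close to the leg, while the hip joints are looser.
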PhysicsAsset /BRCosmetics/Characters/Player/Female/Medium/Bodies/F_MED_RedOasis_ClubDark/Meshes/F_MED_RedOasis_ClubDark_Physics
bodies {
 dyn_shorts_hem_l [{shapeType=Capsule, length=2, radius=4, shapeName=dyn_shorts_hem_l_capsule, center=<6 0 0>, rotation=<rot 90 0 0>}]
 dyn_shorts_hem_r [{shapeType=Capsule, length=2, radius=4, shapeName=dyn_shorts_hem_r_capsule, center=<6 0 0>, rotation=<rot 90 0 0>}]
 pelvis [{shapeType=Capsule, length=17.05, radius=2}]
 thigh_l [{shapeType=Capsule, length=4, radius=4, center=<-10 0 0>, rotation=<rot 90 0 0>}]
 thigh_r [{shapeType=Capsule, length=4, radius=4, center=<10 0 0>, rotation=<rot 90 0 0>}]
}
constraints [
 {bone1Name=thigh_l, bone2Name=pelvis, swing1LimitDegrees=45, swing2LimitDegrees=45, twistLimitDegrees=45}
 {bone1Name=thigh_r, bone2Name=pelvis, swing1LimitDegrees=45, swing2LimitDegrees=45, twistLimitDegrees=45}
 {bone1Name=dyn_shorts_hem_r, bone2Name=thigh_r, swing1LimitDegrees=3, swing2LimitDegrees=8, twistLimitDegrees=12}
 {bone1Name=dyn_shorts_hem_l, bone2Name=thigh_l, swing1LimitDegrees=3, swing2LimitDegrees=8, twistLimitDegrees=12}
]
