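# Post-process-style animation Blueprint for the F_MED_RedOasis_ClubDark female body cosmetic, adding secondary motion to a pose copied from the attached parent mesh.
# The copied pose runs through a Control Rig, then a low-alpha RigidBody simulation, then shorts-hem bone scaling that blends in during 2D movement.
# Each update feeds simulation gravity from a deprecated override function and computes thigh-rotation values that the shown AnimGraph never reads.
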AnimBlueprint /BRCosmetics/Characters/Player/Female/Medium/Bodies/F_MED_RedOasis_ClubDark/Meshes/F_MED_RedOasis_ClubDark_AnimBP
target_skeleton=&/F_MED_RedOasis_ClubDark_Skeleton
graphs [
 AnimGraph {nodes [
   n1 = Control Rig(Source=Copy Pose From Mesh() {bUseAttachedParent=true}) {
    alphaBoolBlend {blendOption=Linear}
    controlRigAssetReference {blueprintRigClass=&/F_MED_RedOasis_ClubDark_ControlRig.F_MED_RedOasis_ClubDark_ControlRig_C}
   }
   n2 = RigidBody!(ComponentPose=Local To Component(LocalPose=@n1), Alpha=0.2, OverrideWorldGravity:Vector=Get GravityOverride(), ExternalForce) -> (Pose) {
    bOverrideWorldGravity=true
    overridePhysicsAsset=&/F_MED_RedOasis_ClubDark_Physics
    alphaBoolBlend {blendOption=Linear}
    baseBoneRef {boneName=pelvis}
    componentAppliedLinearAccClamp=<3000 3000 3000>
    componentLinearAccScale=<0.3 0.3 0.3>
    componentLinearVelScale=<0.3 0.3 0.3>
    overrideWorldGravity {z=-980}
    simSpaceSettings {
     maxAngularAcceleration=1000
     maxAngularVelocity=3
     maxLinearAcceleration=700
     maxLinearVelocity=500
     velocityScaleZ=0
     externalLinearDragV=<0.2 0.2 0.2>
    }
   }
   n3 = Get bIsMoving2D()
   n4 = "Transform (Modify) Bone - Bone: dyn_shorts_hem_l"!(ComponentPose=@n2, bAlphaBoolEnabled:bool=@n3, Translation="0,0,0", Rotation="0,0,0", Scale="1.2,1.4,1.2") -> (Pose) {
    alphaInputType=Bool
    rotationMode=BMM_Additive
    scaleMode=BMM_Additive
    translationMode=BMM_Additive
    alphaBoolBlend {blendInTime=0.3, blendOption=ExpInOut, blendOutTime=0.3}
    boneToModify {boneName=dyn_shorts_hem_l}
    scale=<1.2 1.4 1.2>
   }
   n5 = "Transform (Modify) Bone - Bone: dyn_shorts_hem_r"!(ComponentPose=@n4, bAlphaBoolEnabled:bool=@n3, Translation="0,0,0", Rotation="0,0,0", Scale="1.2,1.4,1.2") -> (Pose) {
    alphaInputType=Bool
    scaleMode=BMM_Additive
    alphaBoolBlend {blendInTime=0.3, blendOption=ExpInOut, blendOutTime=0.3}
    boneToModify {boneName=dyn_shorts_hem_r}
    scale=<1.2 1.4 1.2>
   }
   Output Pose(Result=Component To Local(ComponentPose=@n5)) {layerGroup=None}
  ]}
 EventGraph {nodes [
   Event BlueprintUpdateAnimation()
   n1 = BreakTransform(InTransform=GetDeltaTransformFromRefPose(self=GetOwnerSkeletalMeshComponent(), BoneName=thigh_l, BaseName=pelvis))
   > Set ThighLeftRotationAlpha(ThighLeftRotationAlpha:double=MapRangeClamped(Value=BreakRotator(InRot=@n1.Rotation).Yaw, InRangeA=-20, InRangeB=-90, OutRangeB=0.7)) -> (Output_Get:double)
   n2 = GetDeltaTransformFromRefPose!(self=GetOwnerSkeletalMeshComponent(), BoneName=thigh_r, BaseName=pelvis) -> (ReturnValue_Location:Vector, ReturnValue_Rotation:Rotator, ReturnValue_Scale:Vector)
   > Set ThighRightRotation(ThighRightRotation:double=MapRangeClamped(Value=BreakRotator(InRot=@n2.ReturnValue_Rotation).Yaw, InRangeA=0, InRangeB=-90, OutRangeB=0.7)) -> (Output_Get:double)
   > n3 = Gravity Override State Function DEPRECATED(CharacterPartAnimInstance=Self-Reference(), GameTimeInSeconds=GetGameTimeInSeconds(), GravityOverrideParams=Get Gravity Override Params())
   > Set GravityOverride(GravityOverride=@n3)
  ]}
]
variables [
 Gravity Override Params {type=GravityOverrideParamsStruct}
 GravityOverride {type=Vector}
 ThighLeftRotationAlpha {type=number}
 ThighRightRotation {type=number}
]
